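# Post-process style AnimBlueprint for the Militia Tomato mascot body, adding secondary motion to its collar, armor, pouches, straps, belt and knee pads.
# Copies the attached parent mesh's pose, then runs one chain of weighted bone constraints, Look At and spring-driven Anim Dynamics nodes.
# The event graph maps head rotation relative to spine_05 into alphas that tilt the collar.
# Crouch begin and end events toggle an alpha for thigh armor and hip pouch offsets.
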AnimBlueprint /BRCosmetics/Characters/Player/Male/Medium/Bodies/M_MED_Mascot_Militia_Tomato/Meshes/M_MED_Mascot_Militia_Tomato_AnimBP
target_skeleton=&/M_MED_Mascot_Militia_Tomato_Skeleton
graphs [
 AnimGraph {nodes [
   n1 = "Transform (Modify) Bone - Bone: collarPivot_1_c"(
    ComponentPose=Local To Component(LocalPose=Copy Pose From Mesh() {bUseAttachedParent=true})
    Alpha=Get head_rot_back() -> (head_rot_back:double)
    Translation="0,0,0"
    Rotation="0,45,0"
    Scale="1,1,1"
   ) {
    rotationMode=BMM_Additive
    rotationSpace=BCS_BoneSpace
    alphaBoolBlend {blendOption=Linear}
    alphaScaleBiasClamp {bInterpResult=true}
    boneToModify {boneName=collarPivot_1_c}
    rotation {yaw=45}
   }
   n2 = "Transform (Modify) Bone - Bone: collarPivot_1_c"(ComponentPose=@n1, Alpha=Get head_rot_right() -> (head_rot_right:double), Translation="0,0,0", Rotation="-12,0,0", Scale="1,1,1") {
    rotationMode=BMM_Additive
    rotationSpace=BCS_BoneSpace
    alphaBoolBlend {blendOption=Linear}
    boneToModify {boneName=collarPivot_1_c}
    rotation {pitch=-12}
   }
   n3 = "Transform (Modify) Bone - Bone: collarPivot_1_c"(ComponentPose=@n2, Alpha=Get head_rot_left() -> (head_rot_left:double), Translation="0,0,0", Rotation="12,0,0", Scale="1,1,1") {
    rotationMode=BMM_Additive
    rotationSpace=BCS_BoneSpace
    alphaBoolBlend {blendOption=Linear}
    boneToModify {boneName=collarPivot_1_c}
    rotation {pitch=12}
   }
   n4 = Anim Dynamics - Bone: collarPivot_2_c(ComponentPose=@n3, Alpha=1, ExternalForce="0,0,0") {
    angularDampingOverride=0.9999
    angularSpringConstant=300
    bAngularSpring=true
    bGravityOverrideInSimSpace=true
    bOverrideAngularDamping=true
    gravityScale=0.1
    simulationSpace=RootRelative
    alphaBoolBlend {blendOption=Linear}
    alphaScaleBiasClamp {scale=0.2}
    boundBone {boneName=collarPivot_2_c}
    componentAppliedLinearAccClamp=<0 0 0>
    physicsBodyDefinitions [{
      sphereCollisionRadius=0
      boundBone {boneName=collarPivot_2_c}
      boxExtents=<5 15 15>
      constraintSetup {angularLimitsMax=<0 10 30>, angularLimitsMin=<0 -10 -60>, angularTarget=<1 0 0>}
      localJointOffset=<8 0 0>
     }]
   }
   n5 = "Constraint - thigh_armor_l (PPP)"!(ComponentPose=@n4, Alpha=1, ConstraintWeights_0=0, ConstraintWeights_1:float=0.8, ConstraintWeights_2:float=0.2) -> (Pose) {
    alphaBoolBlend {blendOption=Linear}
    boneToModify {boneName=thigh_armor_l}
    constraintSetup [{targetBone {boneName=thigh_twist_01_l}}, {targetBone {boneName=thigh_l}}, {targetBone {boneName=pelvis_l}}]
    constraintWeights [1,1,1]
   }
   n6 = Get crouch_alpha() -> (crouch_alpha:double)
   n7 = "Transform (Modify) Bone - Bone: thigh_armor_l"(ComponentPose=@n5, Alpha=@n6, Translation="0,0,0", Rotation="0,0,-30", Scale="1,1,1") {
    rotationMode=BMM_Additive
    rotationSpace=BCS_BoneSpace
    alphaBoolBlend {blendOption=Linear}
    alphaScaleBiasClamp {bInterpResult=true}
    boneToModify {boneName=thigh_armor_l}
    rotation {roll=-30}
   }
   n8 = Anim Dynamics - Bone: thigh_armor_l(ComponentPose=@n7, Alpha=1, ExternalForce="0,0,0") {
    angularDampingOverride=1
    angularSpringConstant=1500
    bAngularSpring=true
    bGravityOverrideInSimSpace=true
    bOverrideAngularDamping=true
    gravityScale=0.35
    simulationSpace=RootRelative
    alphaBoolBlend {blendOption=Linear}
    alphaScaleBiasClamp {scale=0.3}
    boundBone {boneName=thigh_armor_l}
    physicsBodyDefinitions [{
      sphereCollisionRadius=0
      boundBone {boneName=thigh_armor_l}
      boxExtents=<10 20 20>
      constraintSetup {angularLimitsMax=<0 10 0>, angularLimitsMin=<0 -10 -40>, angularTarget=<0 0 0>}
      localJointOffset=<10 0 0>
     }]
   }
   n9 = "Transform (Modify) Bone - Bone: thigh_armor_fnt_l"(ComponentPose=@n8, Alpha=@n6, Translation="0,0,0", Rotation="0,0,-30", Scale="1,1,1") {
    rotationMode=BMM_Additive
    rotationSpace=BCS_BoneSpace
    alphaBoolBlend {blendOption=Linear}
    alphaScaleBiasClamp {bInterpResult=true}
    boneToModify {boneName=thigh_armor_fnt_l}
    rotation {roll=-30}
   }
   n10 = Anim Dynamics - Bone: thighPouch_2_l(ComponentPose=@n9, Alpha=1, ExternalForce="0,0,0") {
    angularDampingOverride=1
    angularSpringConstant=1500
    bAngularSpring=true
    bGravityOverrideInSimSpace=true
    bOverrideAngularDamping=true
    gravityScale=0.35
    simulationSpace=RootRelative
    alphaBoolBlend {blendOption=Linear}
    alphaScaleBiasClamp {scale=0}
    boundBone {boneName=thighPouch_2_l}
    physicsBodyDefinitions [{
      sphereCollisionRadius=0
      boundBone {boneName=thighPouch_2_l}
      boxExtents=<10 20 20>
      constraintSetup {angularLimitsMax=<0 0 0>, angularLimitsMin=<0 0 -40>, angularTarget=<0 0 0>}
      localJointOffset=<10 0 0>
     }]
   }
   n11 = Anim Dynamics - Bone: thighPouch_1_l(ComponentPose=@n10, Alpha=1, ExternalForce="0,0,0") {
    angularDampingOverride=1
    angularSpringConstant=1500
    bAngularSpring=true
    bGravityOverrideInSimSpace=true
    bOverrideAngularDamping=true
    gravityScale=0.35
    simulationSpace=RootRelative
    alphaBoolBlend {blendOption=Linear}
    alphaScaleBiasClamp {scale=0}
    boundBone {boneName=thighPouch_1_l}
    physicsBodyDefinitions [{
      sphereCollisionRadius=0
      boundBone {boneName=thighPouch_1_l}
      boxExtents=<10 20 20>
      constraintSetup {angularLimitsMax=<0 0 0>, angularLimitsMin=<0 0 -40>, angularTarget=<0 0 0>}
      localJointOffset=<10 0 0>
     }]
   }
   n12 = "Constraint - thigh_armor_r (P)"(ComponentPose=@n11, Alpha=1, ConstraintWeights_0=1) {
    alphaBoolBlend {blendOption=Linear}
    boneToModify {boneName=thigh_armor_r}
    constraintSetup [{targetBone {boneName=thigh_twist_01_r}}]
    constraintWeights [1]
   }
   n13 = "Constraint - thigh_armor_r (PPP)"!(ComponentPose=@n12, Alpha=1, ConstraintWeights_0=0.4, ConstraintWeights_1:float=0.4, ConstraintWeights_2:float=0.2) -> (Pose) {
    alphaBoolBlend {blendOption=Linear}
    boneToModify {boneName=thigh_armor_r}
    constraintSetup [{targetBone {boneName=thigh_twist_01_r}}, {targetBone {boneName=thigh_r}}, {targetBone {boneName=pelvis_r}}]
    constraintWeights [1,1,1]
   }
   n14 = Get crouch_alpha()
   n15 = "Transform (Modify) Bone - Bone: thigh_armor_r"(ComponentPose=@n13, Alpha=@n14, Translation="0,0,0", Rotation="0,-10,0", Scale="1,1,1") {
    rotationMode=BMM_Additive
    rotationSpace=BCS_BoneSpace
    alphaBoolBlend {blendOption=Linear}
    alphaScaleBiasClamp {bInterpResult=true}
    boneToModify {boneName=thigh_armor_r}
    rotation {yaw=-10}
   }
   n16 = Anim Dynamics - Bone: thigh_armor_r(ComponentPose=@n15, Alpha=1, ExternalForce="0,0,0") {
    angularDampingOverride=1
    angularSpringConstant=1500
    bAngularSpring=true
    bGravityOverrideInSimSpace=true
    bOverrideAngularDamping=true
    gravityScale=0.35
    simulationSpace=RootRelative
    alphaBoolBlend {blendOption=Linear}
    alphaScaleBiasClamp {scale=0.3}
    boundBone {boneName=thigh_armor_r}
    physicsBodyDefinitions [{
      sphereCollisionRadius=0
      boundBone {boneName=thigh_armor_r}
      boxExtents=<10 20 20>
      constraintSetup {angularLimitsMax=<0 10 0>, angularLimitsMin=<0 -10 -40>, angularTarget=<0 0 0>}
      localJointOffset=<10 0 0>
     }]
   }
   n17 = "Transform (Modify) Bone - Bone: thigh_armor_fnt_r"(ComponentPose=@n16, Alpha=@n14, Translation="0,0,0", Rotation="0,0,30", Scale="1,1,1") {
    rotationMode=BMM_Additive
    rotationSpace=BCS_BoneSpace
    alphaBoolBlend {blendOption=Linear}
    alphaScaleBiasClamp {bInterpResult=true}
    boneToModify {boneName=thigh_armor_fnt_r}
    rotation {roll=30}
   }
   n18 = "Constraint - hipPouch_3_r_zero (PPPPPPPP)"!(
    ComponentPose=@n17
    Alpha=1
    ConstraintWeights_0=0.0675
    ConstraintWeights_1:float=0.0704
    ConstraintWeights_2:float=0.0481
    ConstraintWeights_3:float=0.1277
    ConstraintWeights_4:float=0.0835
    ConstraintWeights_5:float=0.108
    ConstraintWeights_6:float=0.0518
    ConstraintWeights_7:float=0.4431
   ) -> (Pose) {
    alphaBoolBlend {blendOption=Linear}
    boneToModify {boneName=hipPouch_3_r_zero}
    constraintSetup [
     {targetBone {boneName=pelvis_l}}
     {targetBone {boneName=thigh_armor_bck_l}}
     {targetBone {boneName=thigh_armor_r}}
     {targetBone {boneName=pelvis_bck_l}}
     {targetBone {boneName=spine_01}}
     {targetBone {boneName=thigh_twist_01_l}}
     {targetBone {boneName=thigh_armor_l}}
     {targetBone {boneName=pelvis}}
    ]
    constraintWeights [1,1,1,1,1,1,1,1]
   }
   n19 = "Constraint - hipPouch_2_r_zero (PPPPPPPP)"!(
    ComponentPose=@n18
    Alpha=1
    ConstraintWeights_0=0.0675
    ConstraintWeights_1:float=0.0704
    ConstraintWeights_2:float=0.0481
    ConstraintWeights_3:float=0.1277
    ConstraintWeights_4:float=0.0835
    ConstraintWeights_5:float=0.108
    ConstraintWeights_6:float=0.0518
    ConstraintWeights_7:float=0.4431
   ) -> (Pose) {
    alphaBoolBlend {blendOption=Linear}
    boneToModify {boneName=hipPouch_2_r_zero}
    constraintSetup [
     {targetBone {boneName=pelvis_l}}
     {targetBone {boneName=thigh_armor_bck_l}}
     {targetBone {boneName=thigh_armor_r}}
     {targetBone {boneName=pelvis_bck_l}}
     {targetBone {boneName=spine_01}}
     {targetBone {boneName=thigh_twist_01_l}}
     {targetBone {boneName=thigh_armor_l}}
     {targetBone {boneName=pelvis}}
    ]
    constraintWeights [1,1,1,1,1,1,1,1]
   }
   n20 = "Constraint - hipPouch_1_r_zero (PPPPPPPP)"!(
    ComponentPose=@n19
    Alpha=1
    ConstraintWeights_0=0.0686
    ConstraintWeights_1:float=0.0106
    ConstraintWeights_2:float=0.0049
    ConstraintWeights_3:float=0.3123
    ConstraintWeights_4:float=0.211
    ConstraintWeights_5:float=0.1151
    ConstraintWeights_6:float=0.0061
    ConstraintWeights_7:float=0.2715
   ) -> (Pose) {
    alphaBoolBlend {blendOption=Linear}
    boneToModify {boneName=hipPouch_1_r_zero}
    constraintSetup [
     {targetBone {boneName=pelvis_l}}
     {targetBone {boneName=thigh_armor_r}}
     {targetBone {boneName=thigh_twist_01_r}}
     {targetBone {boneName=spine_01}}
     {targetBone {boneName=spine_02}}
     {targetBone {boneName=spine_03}}
     {targetBone {boneName=spine_04}}
     {targetBone {boneName=pelvis}}
    ]
    constraintWeights [1,1,1,1,1,1,1,1]
   }
   n21 = "Transform (Modify) Bone - Bone: hipPouch_3_tgt"(ComponentPose=@n20, Alpha=Get crouch_alpha(), Translation="0,-3,-3", Rotation="0,0,0", Scale="1,1,1") {
    translationMode=BMM_Additive
    translationSpace=BCS_BoneSpace
    alphaBoolBlend {blendOption=Linear}
    alphaScaleBiasClamp {bInterpResult=true}
    boneToModify {boneName=hipPouch_3_tgt}
    translation {y=-3, z=-3}
   }
   n22 = Look At - Bone: hipPouch_1_r(ComponentPose=@n21, Alpha=1) {
    alphaBoolBlend {blendOption=Linear}
    boneToModify {boneName=hipPouch_1_r}
    lookAtLocation {x=0}
    lookAtTarget {boneReference {boneName=hipPouch_1_tgt}}
    lookAt_Axis {axis=<1 0>}
    lookUp_Axis {axis=<0 1>}
   }
   n23 = Look At - Bone: hipPouch_2_r(ComponentPose=@n22, Alpha=1) {
    alphaBoolBlend {blendOption=Linear}
    boneToModify {boneName=hipPouch_2_r}
    lookAtLocation {x=0}
    lookAtTarget {boneReference {boneName=hipPouch_2_tgt}}
    lookAt_Axis {axis=<1 0>}
    lookUp_Axis {axis=<0 1>}
   }
   n24 = Look At - Bone: hipPouch_3_r(ComponentPose=@n23, Alpha=1) {
    alphaBoolBlend {blendOption=Linear}
    boneToModify {boneName=hipPouch_3_r}
    lookAtLocation {x=0}
    lookAtTarget {boneReference {boneName=hipPouch_3_tgt}}
    lookAt_Axis {axis=<1 0>}
    lookUp_Axis {axis=<0 1>}
   }
   n25 = Anim Dynamics - Bone: hipPouch_1_r(ComponentPose=@n24, Alpha=1, ExternalForce="0,0,0") {
    angularDampingOverride=1
    angularSpringConstant=1500
    bAngularSpring=true
    bGravityOverrideInSimSpace=true
    bOverrideAngularDamping=true
    gravityScale=0.35
    simulationSpace=RootRelative
    alphaBoolBlend {blendOption=Linear}
    boundBone {boneName=hipPouch_1_r}
    physicsBodyDefinitions [{
      sphereCollisionRadius=0
      boundBone {boneName=hipPouch_1_r}
      boxExtents=<10 20 20>
      constraintSetup {angularLimitsMax=<0 0 0>, angularLimitsMin=<0 0 -40>, angularTarget=<0 0 0>}
      localJointOffset=<10 0 0>
     }]
   }
   n26 = Anim Dynamics - Bone: hipPouch_2_r(ComponentPose=@n25, Alpha=1, ExternalForce="0,0,0") {
    angularDampingOverride=1
    angularSpringConstant=1500
    bAngularSpring=true
    bGravityOverrideInSimSpace=true
    bOverrideAngularDamping=true
    gravityScale=0.35
    simulationSpace=RootRelative
    alphaBoolBlend {blendOption=Linear}
    boundBone {boneName=hipPouch_2_r}
    physicsBodyDefinitions [{
      sphereCollisionRadius=0
      boundBone {boneName=hipPouch_2_r}
      boxExtents=<10 20 20>
      constraintSetup {angularLimitsMax=<0 0 0>, angularLimitsMin=<0 0 -40>, angularTarget=<0 0 0>}
      localJointOffset=<10 0 0>
     }]
   }
   n27 = Anim Dynamics - Bone: hipPouch_3_r(ComponentPose=@n26, Alpha=1, ExternalForce="0,0,0") {
    angularDampingOverride=1
    angularSpringConstant=1500
    bAngularSpring=true
    bGravityOverrideInSimSpace=true
    bOverrideAngularDamping=true
    gravityScale=0.35
    simulationSpace=RootRelative
    alphaBoolBlend {blendOption=Linear}
    boundBone {boneName=hipPouch_3_r}
    physicsBodyDefinitions [{
      sphereCollisionRadius=0
      boundBone {boneName=hipPouch_3_r}
      boxExtents=<10 20 20>
      constraintSetup {angularLimitsMax=<0 0 0>, angularLimitsMin=<0 0 -40>, angularTarget=<0 0 0>}
      localJointOffset=<10 0 0>
     }]
   }
   n28 = Anim Dynamics - Bone: chestPouch_1_r(ComponentPose=@n27, Alpha=1, ExternalForce="0,0,0") {
    angularDampingOverride=1
    angularSpringConstant=1500
    bAngularSpring=true
    bGravityOverrideInSimSpace=true
    bOverrideAngularDamping=true
    gravityScale=0.35
    simulationSpace=RootRelative
    alphaBoolBlend {blendOption=Linear}
    alphaScaleBiasClamp {scale=0.3}
    boundBone {boneName=chestPouch_1_r}
    physicsBodyDefinitions [{
      sphereCollisionRadius=0
      boundBone {boneName=chestPouch_1_r}
      boxExtents=<10 20 20>
      constraintSetup {angularLimitsMax=<0 0 0>, angularLimitsMin=<0 0 -40>, angularTarget=<0 0 0>}
      localJointOffset=<10 0 0>
     }]
   }
   n29 = Anim Dynamics - Bone: chestPouch_2_r(ComponentPose=@n28, Alpha=1, ExternalForce="0,0,0") {
    angularDampingOverride=1
    angularSpringConstant=1500
    bAngularSpring=true
    bGravityOverrideInSimSpace=true
    bOverrideAngularDamping=true
    gravityScale=0.35
    simulationSpace=RootRelative
    alphaBoolBlend {blendOption=Linear}
    boundBone {boneName=chestPouch_2_r}
    physicsBodyDefinitions [{
      sphereCollisionRadius=0
      boundBone {boneName=chestPouch_2_r}
      boxExtents=<10 20 20>
      constraintSetup {angularLimitsMax=<0 0 0>, angularLimitsMin=<0 0 -40>, angularTarget=<0 0 0>}
      localJointOffset=<10 0 0>
     }]
   }
   n30 = Anim Dynamics - Bone: chestPouch_c(ComponentPose=@n29, Alpha=1, ExternalForce="0,0,0") {
    angularDampingOverride=1
    angularSpringConstant=1500
    bAngularSpring=true
    bGravityOverrideInSimSpace=true
    bOverrideAngularDamping=true
    gravityScale=0.35
    simulationSpace=RootRelative
    alphaBoolBlend {blendOption=Linear}
    alphaScaleBiasClamp {scale=0.3}
    boundBone {boneName=chestPouch_c}
    physicsBodyDefinitions [{
      sphereCollisionRadius=0
      boundBone {boneName=chestPouch_c}
      boxExtents=<10 20 20>
      constraintSetup {angularLimitsMax=<0 0 0>, angularLimitsMin=<0 0 -40>, angularTarget=<0 0 0>}
      localJointOffset=<10 0 0>
     }]
   }
   n31 = Anim Dynamics - Bone: chestPouch_l(ComponentPose=@n30, Alpha=1, ExternalForce="0,0,0") {
    angularDampingOverride=1
    angularSpringConstant=1500
    bAngularSpring=true
    bGravityOverrideInSimSpace=true
    bOverrideAngularDamping=true
    gravityScale=0.35
    simulationSpace=RootRelative
    alphaBoolBlend {blendOption=Linear}
    alphaScaleBiasClamp {scale=0.3}
    boundBone {boneName=chestPouch_l}
    physicsBodyDefinitions [{
      sphereCollisionRadius=0
      boundBone {boneName=chestPouch_l}
      boxExtents=<10 20 20>
      constraintSetup {angularLimitsMax=<0 0 0>, angularLimitsMin=<0 0 -40>, angularTarget=<0 0 0>}
      localJointOffset=<10 0 0>
     }]
   }
   n32 = Anim Dynamics - Bone: calf_tassle_r(ComponentPose=@n31, Alpha=1, ExternalForce="0,0,0") {
    angularDampingOverride=1
    angularSpringConstant=1500
    bAngularSpring=true
    bGravityOverrideInSimSpace=true
    bOverrideAngularDamping=true
    gravityScale=0.35
    simulationSpace=RootRelative
    alphaBoolBlend {blendOption=Linear}
    boundBone {boneName=calf_tassle_r}
    physicsBodyDefinitions [{
      sphereCollisionRadius=0
      boundBone {boneName=calf_tassle_r}
      boxExtents=<10 20 20>
      constraintSetup {angularLimitsMax=<0 0 0>, angularLimitsMin=<0 0 -40>, angularTarget=<0 0 0>}
      localJointOffset=<10 0 0>
     }]
   }
   n33 = "Constraint - kneePad_l (RR)"!(ComponentPose=@n32, Alpha=1, ConstraintWeights_0=0.5, ConstraintWeights_1:float=0.5) -> (Pose) {
    alphaBoolBlend {blendOption=Linear}
    boneToModify {boneName=kneePad_l}
    constraintSetup [{transformType=Rotation, targetBone {boneName=thigh_l}}, {transformType=Rotation, targetBone {boneName=calf_l}}]
    constraintWeights [1,1]
   }
   n34 = "Constraint - kneePad_r (RR)"!(ComponentPose=@n33, Alpha=1, ConstraintWeights_0=0.5, ConstraintWeights_1:float=0.5) -> (Pose) {
    alphaBoolBlend {blendOption=Linear}
    boneToModify {boneName=kneePad_r}
    constraintSetup [{transformType=Rotation, targetBone {boneName=thigh_r}}, {transformType=Rotation, targetBone {boneName=calf_r}}]
    constraintWeights [1,1]
   }
   n35 = "Constraint - scapStrap_l (PPPP)"!(
    ComponentPose=@n34
    Alpha=1
    ConstraintWeights_0=0.1971
    ConstraintWeights_1:float=0.0448
    ConstraintWeights_2:float=0.0404
    ConstraintWeights_3:float=0.7177
   ) -> (Pose) {
    alphaBoolBlend {blendOption=Linear}
    boneToModify {boneName=scapStrap_l}
    constraintSetup [{targetBone {boneName=spine_03}}, {targetBone {boneName=spine_04}}, {targetBone {boneName=spine_05}}, {targetBone {boneName=clavicle_l}}]
    constraintWeights [1,1,1,1]
   }
   n36 = "Constraint - scapStrap_r (PPPPP)"!(
    ComponentPose=@n35
    Alpha=1
    ConstraintWeights_0=0.6027
    ConstraintWeights_1:float=0.104
    ConstraintWeights_2:float=0.1991
    ConstraintWeights_3:float=0.0526
    ConstraintWeights_4:float=0.0416
   ) -> (Pose) {
    alphaBoolBlend {blendOption=Linear}
    boneToModify {boneName=scapStrap_r}
    constraintSetup [
     {targetBone {boneName=clavicle_r}}
     {targetBone {boneName=clavicle_tomato_r}}
     {targetBone {boneName=spine_03}}
     {targetBone {boneName=spine_04}}
     {targetBone {boneName=spine_05}}
    ]
    constraintWeights [1,1,1,1,1]
   }
   n37 = Anim Dynamics - Bone: scapStrap_l(ComponentPose=@n36, Alpha=1, ExternalForce="0,0,0") {
    angularDampingOverride=1
    angularSpringConstant=1500
    bAngularSpring=true
    bGravityOverrideInSimSpace=true
    bOverrideAngularDamping=true
    gravityScale=0.35
    simulationSpace=RootRelative
    alphaBoolBlend {blendOption=Linear}
    boundBone {boneName=scapStrap_l}
    physicsBodyDefinitions [{
      sphereCollisionRadius=0
      boundBone {boneName=scapStrap_l}
      boxExtents=<10 20 20>
      constraintSetup {angularLimitsMax=<0 10 0>, angularLimitsMin=<0 -10 -40>, angularTarget=<0 0 0>}
      localJointOffset=<10 0 0>
     }]
   }
   n38 = Anim Dynamics - Bone: scapStrap_r(ComponentPose=@n37, Alpha=1, ExternalForce="0,0,0") {
    angularDampingOverride=1
    angularSpringConstant=1500
    bAngularSpring=true
    bGravityOverrideInSimSpace=true
    bOverrideAngularDamping=true
    gravityScale=0.35
    simulationSpace=RootRelative
    alphaBoolBlend {blendOption=Linear}
    boundBone {boneName=scapStrap_r}
    physicsBodyDefinitions [{
      sphereCollisionRadius=0
      boundBone {boneName=scapStrap_r}
      boxExtents=<10 20 20>
      constraintSetup {angularLimitsMax=<0 10 0>, angularLimitsMin=<0 -10 -40>, angularTarget=<0 0 0>}
      localJointOffset=<10 0 0>
     }]
   }
   n39 = "Constraint - chestStrap_long_1_r (PPPPPP)"!(
    ComponentPose=@n38
    Alpha=1
    ConstraintWeights_0=0.1072
    ConstraintWeights_1:float=0.0271
    ConstraintWeights_2:float=0.0776
    ConstraintWeights_3:float=0.144
    ConstraintWeights_4:float=0.207
    ConstraintWeights_5:float=0.437
   ) -> (Pose) {
    alphaBoolBlend {blendOption=Linear}
    boneToModify {boneName=chestStrap_long_1_r}
    constraintSetup [
     {targetBone {boneName=upperarm_twist_01_r}}
     {targetBone {boneName=spine_02}}
     {targetBone {boneName=spine_03}}
     {targetBone {boneName=spine_04}}
     {targetBone {boneName=spine_05}}
     {targetBone {boneName=clavicle_r}}
    ]
    constraintWeights [1,1,1,1,1,1]
   }
   n40 = "Constraint - chestStrap_long_2_r (PPPPPP)"!(
    ComponentPose=@n39
    Alpha=1
    ConstraintWeights_0=0.0298
    ConstraintWeights_1:float=0.092
    ConstraintWeights_2:float=0.1875
    ConstraintWeights_3:float=0.2754
    ConstraintWeights_4:float=0.1153
    ConstraintWeights_5:float=0.3001
   ) -> (Pose) {
    alphaBoolBlend {blendOption=Linear}
    boneToModify {boneName=chestStrap_long_2_r}
    constraintSetup [
     {targetBone {boneName=upperarm_twist_01_r}}
     {targetBone {boneName=spine_02}}
     {targetBone {boneName=spine_03}}
     {targetBone {boneName=spine_04}}
     {targetBone {boneName=spine_05}}
     {targetBone {boneName=clavicle_r}}
    ]
    constraintWeights [1,1,1,1,1,1]
   }
   n41 = "Constraint - chestStrap_long_3_r (PPPPPP)"!(
    ComponentPose=@n40
    Alpha=1
    ConstraintWeights_0=0.0468
    ConstraintWeights_1:float=0.1485
    ConstraintWeights_2:float=0.3009
    ConstraintWeights_3:float=0.2567
    ConstraintWeights_4:float=0.052
    ConstraintWeights_5:float=0.1951
   ) -> (Pose) {
    alphaBoolBlend {blendOption=Linear}
    boneToModify {boneName=chestStrap_long_3_r}
    constraintSetup [
     {targetBone {boneName=spine_01}}
     {targetBone {boneName=spine_02}}
     {targetBone {boneName=spine_03}}
     {targetBone {boneName=spine_04}}
     {targetBone {boneName=spine_05}}
     {targetBone {boneName=clavicle_r}}
    ]
    constraintWeights [1,1,1,1,1,1]
   }
   n42 = "Constraint - chestStrap_long_4_r (PPPPPPP)"!(
    ComponentPose=@n41
    Alpha=1
    ConstraintWeights_0=0.0956
    ConstraintWeights_1:float=0.2241
    ConstraintWeights_2:float=0.3633
    ConstraintWeights_3:float=0.1693
    ConstraintWeights_4:float=0.0225
    ConstraintWeights_5:float=0.0884
    ConstraintWeights_6:float=0.0367
   ) -> (Pose) {
    alphaBoolBlend {blendOption=Linear}
    boneToModify {boneName=chestStrap_long_4_r}
    constraintSetup [
     {targetBone {boneName=spine_01}}
     {targetBone {boneName=spine_02}}
     {targetBone {boneName=spine_03}}
     {targetBone {boneName=spine_04}}
     {targetBone {boneName=spine_05}}
     {targetBone {boneName=clavicle_r}}
     {targetBone {boneName=pelvis}}
    ]
    constraintWeights [1,1,1,1,1,1,1]
   }
   n43 = "Constraint - chestStrap_long_5_r (PPPPPP)"!(
    ComponentPose=@n42
    Alpha=1
    ConstraintWeights_0=0.18
    ConstraintWeights_1:float=0.2858
    ConstraintWeights_2:float=0.3163
    ConstraintWeights_3:float=0.0829
    ConstraintWeights_4:float=0.0344
    ConstraintWeights_5:float=0.1006
   ) -> (Pose) {
    alphaBoolBlend {blendOption=Linear}
    boneToModify {boneName=chestStrap_long_5_r}
    constraintSetup [
     {targetBone {boneName=spine_01}}
     {targetBone {boneName=spine_02}}
     {targetBone {boneName=spine_03}}
     {targetBone {boneName=spine_04}}
     {targetBone {boneName=clavicle_r}}
     {targetBone {boneName=pelvis}}
    ]
    constraintWeights [1,1,1,1,1,1]
   }
   n44 = "Constraint - chestStrap_long_6_r (PPPPPP)"!(
    ComponentPose=@n43
    Alpha=1
    ConstraintWeights_0=0.2629
    ConstraintWeights_1:float=0.2552
    ConstraintWeights_2:float=0.2217
    ConstraintWeights_3:float=0.0455
    ConstraintWeights_4:float=0.016
    ConstraintWeights_5:float=0.1986
   ) -> (Pose) {
    alphaBoolBlend {blendOption=Linear}
    boneToModify {boneName=chestStrap_long_6_r}
    constraintSetup [
     {targetBone {boneName=spine_01}}
     {targetBone {boneName=spine_02}}
     {targetBone {boneName=spine_03}}
     {targetBone {boneName=spine_04}}
     {targetBone {boneName=clavicle_r}}
     {targetBone {boneName=pelvis}}
    ]
    constraintWeights [1,1,1,1,1,1]
   }
   n45 = "Constraint - chestStrap_long_7_r (PPPPPP)"!(
    ComponentPose=@n44
    Alpha=1
    ConstraintWeights_0=0.0694
    ConstraintWeights_1:float=0.0107
    ConstraintWeights_2:float=0.3157
    ConstraintWeights_3:float=0.2133
    ConstraintWeights_4:float=0.1164
    ConstraintWeights_5:float=0.2745
   ) -> (Pose) {
    alphaBoolBlend {blendOption=Linear}
    boneToModify {boneName=chestStrap_long_7_r}
    constraintSetup [
     {targetBone {boneName=pelvis_l}}
     {targetBone {boneName=thigh_armor_r}}
     {targetBone {boneName=spine_01}}
     {targetBone {boneName=spine_02}}
     {targetBone {boneName=spine_03}}
     {targetBone {boneName=pelvis}}
    ]
    constraintWeights [1,1,1,1,1,1]
   }
   n46 = "Constraint - belt (PPPPPP)"!(
    ComponentPose=@n45
    Alpha=1
    ConstraintWeights_0=0.0705
    ConstraintWeights_1:float=0.0107
    ConstraintWeights_2:float=0.3162
    ConstraintWeights_3:float=0.2129
    ConstraintWeights_4:float=0.116
    ConstraintWeights_5:float=0.2738
   ) -> (Pose) {
    alphaBoolBlend {blendOption=Linear}
    boneToModify {boneName=belt}
    constraintSetup [
     {targetBone {boneName=pelvis_l}}
     {targetBone {boneName=thigh_armor_r}}
     {targetBone {boneName=spine_01}}
     {targetBone {boneName=spine_02}}
     {targetBone {boneName=spine_03}}
     {targetBone {boneName=pelvis}}
    ]
    constraintWeights [1,1,1,1,1,1]
   }
   Output Pose(Result=Component To Local(ComponentPose=@n46)) {layerGroup=None}
  ]}
 EventGraph {nodes [
   Event BlueprintUpdateAnimation()
   n1 = GetOwnerSkeletalMeshComponent()
   n2 = BreakRotator(InRot=BreakTransform(InTransform=GetDeltaTransformFromRefPose(self=@n1, BoneName=head, BaseName=spine_05)).Rotation)
   > Set head_rot_back(head_rot_back:double=MapRangeClamped(Value=@n2.Yaw, InRangeA=0, InRangeB=-80)) -> (Output_Get:double)
   > Set head_rot_left(head_rot_left:double=MapRangeClamped(Value=@n2.Pitch, InRangeA=0, InRangeB=60)) -> (Output_Get:double)
   > Set head_rot_right(head_rot_right:double=MapRangeClamped(Value=@n2.Pitch, InRangeA=0, InRangeB=-60)) -> (Output_Get:double)
   Event OnBeginCrouching()
   > Set crouch_alpha(crouch_alpha:double=1) -> (Output_Get:double)
   Event OnEndCrouching()
   > Set crouch_alpha(crouch_alpha=0)
   TryGetPawnOwner()
  ]}
]
variables [
 crouch_alpha {type=number}
 head_rot_back {type=number}
 head_rot_left {type=number}
 head_rot_right {type=number}
]
